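# Material function for the Origin Prison player body cosmetic that draws glowing, animated string lines as an opacity-masked emissive layer over black base color.
# Thresholded distance-field textures form the lines, with hashed-time glitch noise mixed in, and Fresnel gradients tint them cyan or purple.
# World position offset combines sine waves, texture-masked inflation and a mirrored local push within a UV mask.
# Curve-named parameters expose line, glitch, color, inflate, chest-deform and looping-progression controls through named reroutes.
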MaterialFunction /BRCosmetics/Characters/Player/Male/Medium/Bodies/M_MED_Origin_Prison/Materials/TechArt/MF_OriginPrison_String
graph {nodes [
  n1 = MaterialFunctionCall() {materialFunction=&BoundingBoxBased_0-1_UVW}
  n2 = SphereMask(A=@n1.R, B=Constant() {r=0.5}, Radius=Constant() {r=0.55}, Hardness=Constant() {r=0.5})
  n3 = Constant() {r=0.5}
  n4 = Constant() {r=0.65}
  n5 = TextureSampleParameter2D() {
   parameterName=DF progression Texture
   samplerSource=SSM_Wrap_WorldGroupSettings
   samplerType=SAMPLERTYPE_LinearGrayscale
   texture=/BRCosmetics/Characters/Player/Male/Medium/Bodies/M_MED_Origin_Prison/Textures/TechArt/T_Origin_Prison_DF
  }
  n6 = Frac(None=Multiply(A=Time()))
  n7 = Constant() {r=0.1}
  n8 = Constant()
  n9 = Time()
  n10 = Subtract(A=Frac(None=Multiply(A=@n9) {constB=0.8})) {constB=0.02}
  n11 = Constant() {r=0.03}
  n12 = Constant()
  n13 = Multiply(
   A=Multiply(A=OneMinus(None=@n2), B=SmoothStep(Min=@n3, Max=@n4, Value=@n1.B))
   B=Add(A=SphereMask(A=@n5.RGB, B=@n6, Radius=@n7, Hardness=@n8), B=SphereMask(A=@n5.RGB, B=@n10, Radius=@n11, Hardness=@n12))
  )
  n14 = Fresnel(ExponentIn=Constant() {r=1})
  n15 = Saturate(None=Multiply(A=Multiply(A=@n13, B=@n14), B=Constant() {r=2}))
  n16 = TextureSampleParameter2D(UVs=NamedRerouteUsage()) {
   group=OriginPrision
   parameterName=DF Texture
   samplerSource=SSM_Wrap_WorldGroupSettings
   samplerType=SAMPLERTYPE_LinearGrayscale
   sortPriority=2
   texture=/BRCosmetics/Characters/Player/Male/Medium/Bodies/M_MED_Origin_Prison/Textures/TechArt/T_Origin_Prison_String_DF
  }
  n17 = NamedRerouteUsage()
  n18 = Multiply(A=Saturate(None=Add(A=@n16.R, B=@n17) {constB=-0.25}), B=Constant() {r=14})
  Add(A=@n15, B=Multiply(A=NamedRerouteUsage(), B=Saturate(None=@n18)))
  Constant() {r=0.99}
  Constant()
  Constant() {r=0.1}
  n19 = Fresnel(ExponentIn=Constant() {r=1})
  n20 = LinearInterpolate(A=Constant3Vector() {constant=<rgba 0.1979 0.7921 1 0>}, B=Constant3Vector() {constant=<rgba 0 0.2693 0.474 0>}, Alpha=@n19)
  n21 = LinearInterpolate(A=Constant3Vector() {constant=<rgba 0.5324 0.4178 0.8021 0>}, B=Constant3Vector() {constant=<rgba 0.1535 0.0907 0.4792 0>}, Alpha=@n19)
  n22 = NamedRerouteUsage()
  n23 = NamedRerouteUsage()
  n24 = Multiply(A=Time())
  n25 = ConstantBiasScale(None=MaterialFunctionCall(Float=Frac(None=@n24)) {materialFunction=&MF_Hash}.Float)
  n26 = Multiply(A=TextureCoordinate()) {constB=1.5}
  n27 = Add(A=Multiply(A=TextureSample(UVs=@n26) {samplerType=SAMPLERTYPE_LinearGrayscale, texture=/Game/Global/Textures/Default/Noise/glitch_noise}.R)) {constB=0.5}
  n28 = Saturate(None=Subtract(A=Saturate(None=@n27)) {constB=0.1})
  n29 = NamedRerouteUsage()
  n30 = LinearInterpolate(A=@n23, B=LinearInterpolate(A=@n23, B=@n28, Alpha=@n25), Alpha=@n29)
  n31 = TextureSampleParameter2D(UVs=NamedRerouteUsage()) {
   group=OriginPrision
   parameterName=DF Texture
   samplerSource=SSM_Wrap_WorldGroupSettings
   samplerType=SAMPLERTYPE_LinearGrayscale
   sortPriority=2
   texture=&T_Origin_Prison_String_DF
  }
  n32 = LinearInterpolate(A=@n31.R, B=LinearInterpolate(A=@n31.R, B=@n28, Alpha=@n25), Alpha=@n29)
  n33 = NamedRerouteUsage()
  n34 = Saturate(None=Add(
    A=LinearInterpolate(A=@n30, B=@n32, Alpha=@n33)
    B=ScalarParameter() {defaultValue=-0.54, desc=Painted Line Thickness, group=OriginPrision, parameterName=Depth Offset}
   ) {constB=-0.25})
  n35 = Constant() {r=14}
  n36 = SetMaterialAttributes!(
   MaterialAttributes=FunctionInput() {inputType=FunctionInput_MaterialAttributes}
   BaseColor=Constant()
   EmissiveColor=Multiply(B=LinearInterpolate(A=@n20, B=@n21, Alpha=@n22)) {constA=1}
   OpacityMask=Saturate(None=Multiply(A=@n34, B=@n35))
   WorldPositionOffset=NamedRerouteUsage()
   Normal=Constant3Vector() {constant=<rgba 0 0 1 0>}
  ) {attributeSetTypes [
    #1
    #2
    #3
    #4
    #5
   ]}
  FunctionOutput(None=@n36)
  MaterialFunctionCall(Alpha Threshold=@n15) {materialFunction=&DitherTemporalAA_Fortnite}
  n37 = Time()
  Multiply(A=ConstantBiasScale(None=Sine(None=@n37))) {constB=0.04}
  n38 = Time()
  Multiply(A=ConstantBiasScale(None=Sine(None=@n38))) {constB=0.04}
  n39 = Time()
  Multiply(A=@n39) {constB=2}
  n40 = Time()
  Multiply(A=Frac(None=Multiply(A=@n40) {constB=0.25})) {constB=0.04}
  Multiply(A=Time())
  n41 = OneMinus(None=NamedRerouteUsage())
  n42 = TextureCoordinate()
  n43 = TextureSample(UVs=Panner(Coordinate=Multiply(A=@n42) {constB=8}) {speedX=0.2, speedY=0.3}) {samplerType=SAMPLERTYPE_Grayscale, texture=&LowResBlurredNoise}
  n44 = Multiply(A=Subtract(A=Multiply(A=@n43.R) {constB=2}) {constB=1.5}) {constB=0.2}
  n45 = Constant() {r=-0.2}
  Multiply(A=Add(A=Multiply(A=@n41, B=@n44)), B=@n45)
  n46 = VertexNormalWS()
  n47 = PreSkinnedPosition()
  n48 = VertexInterpolator(VS=@n47)
  n49 = MaterialFunctionCall(Float3=Multiply(
    A=ShaderStageSwitch(PixelShader=@n48.PS, VertexShader=@n47)
    B=ScalarParameter() {defaultValue=0.02, group=OriginPrision, parameterName=LocalWave_Frequency}
   ) {constB=0.1}) {materialFunction=&BreakOutFloat3Components}
  n50 = Add(A=@n49.R, B=@n49.R)
  n51 = Multiply(A=@n49.B) {constB=3}
  n52 = Subtract(A=Time(), B=ScalarParameter() {group=OriginPrision, parameterName=LocalWave_WPO_random_seed})
  n53 = ScalarParameter() {defaultValue=2.4, group=OriginPrision, parameterName=LocalWave_Speed}
  n54 = Subtract(A=Subtract(A=Add(A=@n50, B=@n51)) {constB=2}, B=Frac(None=Multiply(A=@n52, B=@n53) {constB=0.6}))
  n55 = Multiply(A=Add(A=Sine(None=@n54))) {constB=0.5}
  n56 = Multiply(A=MaterialFunctionCall() {materialFunction=&BoundingBoxBased_0-1_UVW}.B) {constB=16}
  n57 = Multiply(A=Power(Base=Sine(None=@n56)) {constExponent=3}, B=Sine(None=Frac(None=@n39)))
  n58 = Add(
   A=Multiply(A=Multiply(A=@n46, B=@n55), B=NamedRerouteUsage())
   B=RotateAboutAxis(
    NormalizedRotationAxis=Constant4Vector() {constant=<rgba 0 0 1 1>}.RGBA
    RotationAngle=Multiply(A=@n57) {constB=0.01}
    PivotPoint=MaterialFunctionCall() {materialFunction=&ObjectPivotPoint}.Object Pivot Location
    Position=WorldPosition().XYZ
   )
  ) {constB=0}
  n59 = VertexNormalWS()
  n60 = ComponentMask(None=@n59) {g=true, r=true}
  n61 = ComponentMask(None=@n59) {b=true}
  n62 = LinearInterpolate(
   A=@n61
   B=Multiply(B=@n61) {constA=10}
   Alpha=MaterialFunctionCall() {materialFunction=&BoundingBoxBased_0-1_UVW}.B
  )
  n63 = TextureSampleParameter2D() {
   parameterName=DF progression Texture
   samplerSource=SSM_Wrap_WorldGroupSettings
   samplerType=SAMPLERTYPE_LinearGrayscale
   texture=&T_Origin_Prison_DF
  }
  n64 = LinearInterpolate(B=Multiply(A=AppendVector(A=@n60, B=@n62), B=NamedRerouteUsage()), Alpha=@n63.RGB)
  n65 = Constant3Vector() {constant=<rgba 1 0 0 0>}
  n66 = Multiply(B=NamedRerouteUsage() {desc=0-0.03}) {constA=200}
  n67 = Constant3Vector() {constant=<rgba -1 0 0 0>}
  n68 = MaterialFunctionCall(Float2=TextureCoordinate() {coordinateIndex=1}) {materialFunction=&BreakOutFloat2Components}
  n69 = Transform(None=LinearInterpolate(A=Multiply(A=@n65, B=@n66), B=Multiply(A=@n66, B=@n67), Alpha=Step(Y=@n68.R) {constX=0.5})) {transformSourceType=TRANSFORMSOURCE_Local}
  n70 = TextureCoordinate()
  n71 = Multiply(
   A=SphereMask(A=TextureCoordinate(), B=Constant2Vector() {g=0.495, r=0.135}, Radius=Constant() {r=0.12}, Hardness=Constant() {r=0.5})
   B=SphereMask(
    A=MaterialFunctionCall(Float2=@n70) {materialFunction=&BreakOutFloat2Components}.R
    B=Constant() {r=0.134}
    Radius=Constant() {r=0.1}
    Hardness=Constant() {r=0.5}
   )
  )
  NamedRerouteDeclaration(None=Add(A=Add(A=@n58, B=@n64), B=LinearInterpolate(B=@n69, Alpha=@n71))) {name=WPO, nodeColor=<rgba 1 0 0.6094 1>}
  NamedRerouteDeclaration(None=ScalarParameter() {group=OriginPrision, parameterName=MaterialCurveB}) {name=Line Transition Amount, nodeColor=<rgba 0.7578 0 1 1>}
  NamedRerouteDeclaration(None=ScalarParameter() {group=OriginPrision, parameterName=HT_L1_TransitionCurve}) {name=Glitch Amount, nodeColor=<rgba 0.3594 0 1 1>}
  NamedRerouteDeclaration(None=ScalarParameter() {defaultValue=0.5, group=OriginPrision, parameterName=JH_TransformCurve}) {name=Inflate Amount, nodeColor=<rgba 1 0.4922 0 1>}
  NamedRerouteDeclaration(None=ScalarParameter() {defaultValue=1, group=OriginPrision, parameterName=Lexa_TransformationCurve}) {name=WPO Noise Scale, nodeColor=<rgba 0.2422 1 0 1>}
  NamedRerouteDeclaration(None=ScalarParameter() {group=OriginPrision, parameterName=MaterialCurveD}) {name=Color Transition Amount, nodeColor=<rgba 1 0.2109 0 1>}
  n72 = MaterialFunctionCall(Float2=TextureCoordinate()) {materialFunction=&BreakOutFloat2Components}
  n73 = MaterialFunctionCall(Float2=TextureCoordinate()) {materialFunction=&BreakOutFloat2Components}
  n74 = NamedRerouteUsage() {desc=0-0.03}
  n75 = Constant() {r=0.06}
  n76 = Constant()
  n77 = LinearInterpolate(A=@n73.R, B=Add(A=@n73.R, B=@n74) {constB=0.015}, Alpha=SphereMask(A=@n73.R, B=Constant() {r=0.08}, Radius=@n75, Hardness=@n76))
  n78 = Subtract(A=@n72.R, B=@n74)
  n79 = SphereMask(A=@n72.R, B=Constant() {r=0.188}, Radius=@n75, Hardness=@n76)
  n80 = TextureCoordinate()
  n81 = Constant2Vector() {g=0.505, r=0.135}
  n82 = Constant() {r=0.1}
  n83 = Constant()
  n84 = AppendVector(
   A=LinearInterpolate(A=@n72.R, B=LinearInterpolate(A=@n77, B=@n78, Alpha=@n79), Alpha=SphereMask(A=@n80, B=@n81, Radius=@n82, Hardness=@n83))
   B=@n72.G
  )
  NamedRerouteDeclaration(None=@n84) {name=Deformed UV, nodeColor=<rgba 0.0398 0 0.3281 1>}
  NamedRerouteDeclaration(None=ScalarParameter() {group=OriginPrision, parameterName=MaterialCurveC}) {name=Chest Deform Amount, nodeColor=<rgba 0 1 0.4844 1>}
  n85 = TextureSampleParameter2D() {
   parameterName=DF progression Texture
   samplerType=SAMPLERTYPE_LinearGrayscale
   texture=&T_Origin_Prison_DF
  }
  n86 = MaterialFunctionCall(Input=@n85.R, Input Low=Constant(), Input High=Constant() {r=0.7}, Target Low=Constant(), Target High=Constant() {r=1}) {materialFunction=&RemapValueRange}
  n87 = Constant() {desc=Loop Number, r=11}
  n88 = Multiply(A=Power(Base=Saturate(None=@n86.Result)), B=@n87) {constB=50}
  n89 = Frac(None=NamedRerouteUsage())
  n90 = Constant() {r=0.5}
  n91 = Add(B=@n90) {constA=1}
  n92 = Saturate(None=Subtract(A=Multiply(A=@n89, B=@n91) {constB=1.5}, B=@n90) {constB=0.5})
  n93 = Constant()
  n94 = Frac(None=@n88)
  n95 = Frac(None=@n88)
  n96 = MaterialFunctionCall() {materialFunction=&BoundingBoxBased_0-1_UVW}
  n97 = Constant() {r=0.512}
  n98 = Constant() {r=0.31}
  n99 = Constant() {r=1}
  n100 = LinearInterpolate(
   A=If(A=@n88, B=Multiply(A=@n87, B=@n92), "A > B"=@n93, "A == B"=@n94, "A < B"=@n94)
   B=If(A=@n88, B=Multiply(A=@n87, B=@n89), "A > B"=@n93, "A == B"=@n95, "A < B"=@n95)
   Alpha=OneMinus(None=SphereMask(A=@n96.R, B=@n97, Radius=@n98, Hardness=@n99))
  )
  NamedRerouteDeclaration(None=Saturate(None=Add(A=@n100, B=@n45) {constB=-0.2})) {name=Progression Loops, nodeColor=<rgba 0.4766 0 1 1>}
  NamedRerouteDeclaration(None=ScalarParameter() {group=OriginPrision, parameterName=MaterialCurveA}) {name=DF Loop Progression, nodeColor=<rgba 0.8516 0 1 1>}
  NamedRerouteUsage()
  OneMinus(None=@n14)
  ScalarParameter() {defaultValue=-0.4, group=OriginPrision, parameterName=OG_DistanceOffset}
  ScalarParameter() {defaultValue=-0.4, group=OriginPrision, parameterName=OG_DistanceOffset}
  ScalarParameter() {defaultValue=-0.54, desc=Painted Line Thickness, group=OriginPrision, parameterName=Depth Offset}
  TextureSampleParameter2D(UVs=@n84) {
   group=OriginPrision
   parameterName=DF Texture
   samplerSource=SSM_Wrap_WorldGroupSettings
   samplerType=SAMPLERTYPE_LinearGrayscale
   sortPriority=2
   texture=&T_Origin_Prison_String_DF
  }
 ]}
