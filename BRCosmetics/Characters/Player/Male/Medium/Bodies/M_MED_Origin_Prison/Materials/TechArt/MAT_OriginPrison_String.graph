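# Masked, two-sided skeletal mesh material for the string on the Origin Prison character body, using material attributes with morph target and mesh deformer support.
# BaseFN_CharacterFunction_PreFX feeds MF_OriginPrison_String, whose result passes through BaseFN_CharacterFunction_PostFX into CINE_MF_CharPostFX.
# The string-specific look lives in that material function, wrapped by the shared character stages.
# A CharacterSkinFX call branches off the PreFX Gmap output, outside that chain.
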Material /BRCosmetics/Characters/Player/Male/Medium/Bodies/M_MED_Origin_Prison/Materials/TechArt/MAT_OriginPrison_String
graph {
 properties {
  bUseMaterialAttributes=true
  bUsedWithMeshDeformer=true
  bUsedWithMorphTargets=true
  bUsedWithSkeletalMesh=true
  blendMode=BLEND_Masked
  pixelDepthOffsetMode=PDOM_Legacy
  refractionMethod=RM_IndexOfRefraction
  twoSided=true
  usedShadingModels=Default Lit
 }
 nodes [
  n1 = MaterialFunctionCall() {materialFunction=&BaseFN_CharacterFunction_PreFX}
  n2 = MaterialFunctionCall(
   AO=@n1.AO
   Distance Blend=@n1.AO
   Fresnel=@n1.AO
   Gmap=@n1.AO
   Pre Skinned Local Position=@n1.AO
   Material=MaterialFunctionCall!(In=@n1.Result) {materialFunction=&/MF_OriginPrison_String}.Result
  ) {materialFunction=&BaseFN_CharacterFunction_PostFX}
  MaterialFunctionCall(In=@n2.Result) {materialFunction=&CINE_MF_CharPostFX}
  MaterialFunctionCall(Gmap=@n1.Gmap, AO=@n1.Gmap) {materialFunction=&CharacterSkinFX}
 ]
}
